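# Test widget that lays out item shop offer-image test tiles of several offer types and sizes in a fixed-width WrapBox grid.
# On PreConstruct it loops over the WrapBox children and passes the ShowRulesetDebugText flag to each test tile's ruleset debug toggle.
# When the loop completes, the TileBreak_Padding flag selects between even and taller vertical inner slot padding for the grid.
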WidgetBlueprint /ItemShop/A_TestWidgets/A_OfferImage_TestGrid
graphs [EventGraph {nodes [
   Event PreConstruct()
   > n1 = For Each Loop(Array:Array of Widget Object References=GetAllChildren(self=Get WrapBox_All() -> (WrapBox_All:Wrap Box Object Reference))) -> (Array Element:Widget Object Reference)
   SetInnerSlotPadding!(
    execute=@n1.Completed
    self:Wrap Box Object Reference=Get WrapBox_All()
    InPadding:Vector 2D Structure=Select(
     Option 0:Vector 2D Structure=(X=24,Y=24)
     Option 1:Vector 2D Structure=(X=16,Y=72)
     Index=Get TileBreak_Padding() -> (TileBreak_Padding:bool)
    ) -> (ReturnValue:Vector 2D Structure)
   ) -> (then)
   n2 = Cast To A_OfferImage_TestTile!(execute=@n1.LoopBody, Object=@n1) -> (then, CastFailed, AsA Offer Image Test Tile:A Offer Image Test Tile Object Reference) {is=title K2Node_DynamicCast}
   > ToggleRulesetDebug!(execute, self:A Offer Image Test Tile Object Reference=@n2, ShowRulesetDebug:bool=Get ShowRulesetDebugText() -> (ShowRulesetDebugText:bool)) -> (then)
   Event Construct()
   Event Tick()
  ]}]
variables [
 ShowRulesetDebugText {type=boolean}
 TileBreak_Padding {type=boolean}
]
widgets {
 info {inheritedWidgetCount=0, parentClass=&UserWidget, rootWidgetClass=&SizeBox, widgetCount=8}
 widgets [
  SizeBox_0 {widgetClassPath=&SizeBox, widget=SizeBox_0, properties {bOverride_WidthOverride=true, widthOverride=1096}}
  WrapBox_All {
   widgetClassPath=&WrapBox
   bIsVariable=true
   parent=SizeBox_0
   widget=WrapBox_All
   properties {innerSlotPadding=<16 72>}
   slot {class=SizeBoxSlot}
  }
  A_OfferImage_TestTile {
   widgetClassPath=&/A_OfferImage_TestTile
   bIsVariable=true
   parent=WrapBox_All
   widget=A_OfferImage_TestTile
   properties {offerType="Bundle (Non-jamtrack)", showRulesetDebug=false}
   slot {class=WrapBoxSlot, properties {horizontalAlignment=HAlign_Left, verticalAlignment=VAlign_Top}}
  }
  A_OfferImage_TestTile_4 {
   widgetClassPath=&A_OfferImage_TestTile
   bIsVariable=true
   parent=WrapBox_All
   widget=A_OfferImage_TestTile_4
   properties {offerType=Emote, showRulesetDebug=false, tileSize=Size_2_x_1}
   slot {class=WrapBoxSlot, properties {horizontalAlignment=HAlign_Left, verticalAlignment=VAlign_Top}}
  }
  A_OfferImage_TestTile_1 {
   widgetClassPath=&A_OfferImage_TestTile
   bIsVariable=true
   parent=WrapBox_All
   widget=A_OfferImage_TestTile_1
   properties {offerType="Outfit (Non-minifig)", showRulesetDebug=false, tileSize=Size_1_x_1}
   slot {class=WrapBoxSlot, properties {horizontalAlignment=HAlign_Left, verticalAlignment=VAlign_Top}}
  }
  A_OfferImage_TestTile_2 {
   widgetClassPath=&A_OfferImage_TestTile
   bIsVariable=true
   parent=WrapBox_All
   widget=A_OfferImage_TestTile_2
   properties {offerType="Outfit (Non-minifig)", showRulesetDebug=false, tileSize=Size_1_x_1}
   slot {class=WrapBoxSlot, properties {horizontalAlignment=HAlign_Left, verticalAlignment=VAlign_Top}}
  }
  A_OfferImage_TestTile_3 {
   widgetClassPath=&A_OfferImage_TestTile
   bIsVariable=true
   parent=WrapBox_All
   widget=A_OfferImage_TestTile_3
   properties {offerType="Outfit (Non-minifig)", showRulesetDebug=false, tileSize=Size_1_x_1}
   slot {class=WrapBoxSlot, properties {horizontalAlignment=HAlign_Left, verticalAlignment=VAlign_Top}}
  }
  A_OfferImage_TestTile_5 {
   widgetClassPath=&A_OfferImage_TestTile
   bIsVariable=true
   parent=WrapBox_All
   widget=A_OfferImage_TestTile_5
   properties {offerType="Outfit (Non-minifig)", showRulesetDebug=false, tileSize=Size_1_x_1}
   slot {class=WrapBoxSlot, properties {horizontalAlignment=HAlign_Left, verticalAlignment=VAlign_Top}}
  }
 ]
}
