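# Test item shop tile widget that previews how an offer image looks by tile size and offer type, with placeholder offer info.
# PreConstruct sizes the tile, toggles placeholder smart-card columns, and applies frame-break, texture or material, and background colors to the swiping offer image component.
# Offer type picks a ruleset DataTable; vehicle, smart-card and bundle rows try suffixed names before plain size fallbacks.
# The chosen ruleset feeds the image component and its debug label.
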WidgetBlueprint /ItemShop/A_TestWidgets/A_OfferImage_TestTile_V2
graphs [
 EventGraph {nodes [
   Event PreConstruct()
   > ToggleRulesetDebug!(execute, self, ShowRulesetDebug:bool=Get ShowRulesetDebug() -> (ShowRulesetDebug:bool)) -> (then)
   n1 = "To Vector2D (IntPoint)"!(InIntPoint:Int Point Structure=GetGridSpanFromTileSize!(TileSize:EItemShopTileSize Enum=Get TileSize() -> (TileSize:EItemShopTileSize Enum=Size_1_x_1)) -> (GridSpan:Int Point Structure)) -> (ReturnValue:Vector 2D Structure)
   n2 = vector2d + vector2d!(
    A:Vector 2D Structure=vector2d * vector2d(A:Vector 2D Structure=@n1, B:Vector 2D Structure=(X=256,Y=320)) -> (ReturnValue:Vector 2D Structure) {is=title K2Node_PromotableOperator}
    B:Vector 2D Structure=vector2d * vector2d(
     A=vector2d - vector2d(A:Vector 2D Structure=@n1, B:Vector 2D Structure=(X=1,Y=1)) -> (ReturnValue:Vector 2D Structure) {is=title K2Node_PromotableOperator}
     B=(X=16,Y=0)
    )
   ) -> (ReturnValue_X:double, ReturnValue_Y:double) {is=title K2Node_PromotableOperator}
   > Set WidthOverride!(execute, WidthOverride:float=@n2, self:Size Box Object Reference=Get SizeBox_Tile() -> (SizeBox_Tile:Size Box Object Reference)) -> (then, Output_Get:float)
   > Set HeightOverride!(execute, HeightOverride:float=@n2.ReturnValue_Y, self:Size Box Object Reference=Get SizeBox_Tile()) -> (then, Output_Get:float)
   n3 = Get FrameBreak() -> (FrameBreak:bool)
   n4 = Get WBP_ItemShopTile_SwipingOfferImage_Separate() -> (WBP_ItemShopTile_SwipingOfferImage_Separate:WBP Item Shop Tile Swiping Offer Image Separate Object Reference)
   > Set with Broadcast IsFrameBreaking!(execute, IsFrameBreaking:bool=@n3, self:WBP Item Shop Tile Swiping Offer Image Separate Object Reference=@n4) -> (then, Output_Get:bool)
   n5 = Get FrameBreakPadding() -> (FrameBreakPadding:double)
   > SetPadding!(
    execute
    self:Overlay Slot Object Reference=SlotAsOverlaySlot(Widget=Get Image_Current!(self:WBP Item Shop Tile Swiping Offer Image Separate Object Reference=@n4) -> (Image_Current:Image Object Reference))
    InPadding_Left:float=0
    InPadding_Top:float=SelectFloat(A=float * float(A=@n5, B=-1) {is=title K2Node_PromotableOperator}, B=0, bPickA=@n3)
    InPadding_Right:float=0
    InPadding_Bottom:float=0
   ) -> (then)
   > n6 = For Each Loop(Array:Array of Widget Object References=GetAllChildren(self=Get HorizontalBox_SmartCards() -> (HorizontalBox_SmartCards:Horizontal Box Object Reference))) -> (Array Element:Widget Object Reference)
   n7 = "integer < integer"(A=@n6.Array Index, B=Get SmartCardColumns() -> (SmartCardColumns:int=0))
   n8 = Select!(
    NewEnumerator23:bool=false
    NewEnumerator0:bool=false
    NewEnumerator24:bool=false
    NewEnumerator1:bool=false
    NewEnumerator2:bool=false
    NewEnumerator3:bool=false
    NewEnumerator28:bool=false
    NewEnumerator4:bool=false
    NewEnumerator5:bool=false
    NewEnumerator6:bool=false
    NewEnumerator7:bool=false
    NewEnumerator8:bool=false
    NewEnumerator9:bool=false
    NewEnumerator10:bool=false
    NewEnumerator11:bool=false
    NewEnumerator12:bool=false
    NewEnumerator30:bool=false
    NewEnumerator13:bool=false
    NewEnumerator14:bool=false
    NewEnumerator15:bool=false
    NewEnumerator16:bool=false
    NewEnumerator17:bool=false
    NewEnumerator27:bool=false
    NewEnumerator18:bool=false
    NewEnumerator19:bool=false
    NewEnumerator20:bool=false
    NewEnumerator29:bool=true
    NewEnumerator21:bool=false
    NewEnumerator22:bool=true
    NewEnumerator25:bool=false
    NewEnumerator26:bool=false
    NewEnumerator31:bool=false
    Index:E_ItemShopOfferType Enum=Get OfferType() -> (OfferType:E_ItemShopOfferType Enum=NewEnumerator23)
   ) -> (ReturnValue:bool)
   > Set Visibility(
    Visibility=Select!(Option 0:ESlateVisibility Enum=Collapsed, Option 1:ESlateVisibility Enum=SelfHitTestInvisible, Index:bool=AND Boolean(A=@n7, B=@n8)) -> (ReturnValue:ESlateVisibility Enum)
    self=@n6
   )
   n9 = Get PresentationMaterialOverride() -> (PresentationMaterialOverride:Material Interface Object Reference)
   n10 = IsValid(Object=@n9)
   n11 = Get WBP_ItemShopTile_SwipingOfferImage_Separate()
   SetSinglePresentation!(
    execute=@n6.Completed
    self:WBP Item Shop Tile Swiping Offer Image Separate Object Reference=Get WBP_ItemShopTile_SwipingOfferImage_Separate()
    Presentation_Texture:Texture 2D Object Reference=Select(
     Option 0:Texture 2D Object Reference=Get Texture() -> (Texture:Texture 2D Object Reference)
     Option 1:Texture 2D Object Reference
     Index=@n10
    ) -> (ReturnValue:Texture 2D Object Reference)
    Presentation_Material:Material Interface Object Reference=Select(
     Option 0:Material Interface Object Reference=Get DefaultPresentationMaterial!(self:WBP Item Shop Tile Swiping Offer Image Separate Object Reference=@n11) -> (DefaultPresentationMaterial:Material Interface Object Reference)
     Option 1:Material Interface Object Reference=@n9
     Index=@n10
    ) -> (ReturnValue:Material Interface Object Reference)
   ) -> (then)
   > n12 = GetDynamicMaterial(self:Border Object Reference=Get Border_OfferInfo() -> (Border_OfferInfo:Border Object Reference))
   n13 = Get Text Backing Color() -> (Text Backing Color:Linear Color Structure)
   > SetVectorParameterValue(self=@n12, ParameterName=Color, Value=@n13)
   > SetVectorParameterValue(self=@n12, ParameterName=Color2, Value=@n13)
   > n14 = CreateDynamicMaterialInstance!(
    execute
    Parent=Get BackgroundMaterialOverride() -> (BackgroundMaterialOverride:Material Interface Object Reference)
    OptionalName=None
    CreationFlags=None
   ) -> (then, ReturnValue)
   > SetBackgroundMaterial!(
    execute
    self:WBP Item Shop Tile Swiping Offer Image Separate Object Reference=Get WBP_ItemShopTile_SwipingOfferImage_Separate()
    Material:Material Interface Object Reference=@n14
   ) -> (then)
   > SetTileBackgroundColors!(
    execute
    MID:Material Instance Dynamic Object Reference=@n14
    FortItemShopOfferMetaDisplayData_Color1:Linear Color Structure=Get BackgroundColor1() -> (BackgroundColor1:Linear Color Structure)
    FortItemShopOfferMetaDisplayData_Color2:Linear Color Structure=Get BackgroundColor2() -> (BackgroundColor2:Linear Color Structure)
    FortItemShopOfferMetaDisplayData_Color3:Linear Color Structure=Select(
     Option 0:Linear Color Structure=(R=0,G=0,B=0,A=0)
     Option 1:Linear Color Structure=Get BackgroundColor3() -> (BackgroundColor3:Linear Color Structure)
     Index=Get Use3BackgroundColors() -> (Use3BackgroundColors:bool)
    ) -> (ReturnValue:Linear Color Structure)
    FortItemShopOfferMetaDisplayData_TextBackgroundColor:Linear Color Structure=(R=0,G=0,B=0,A=0)
    FortItemShopOfferMetaDisplayData_TileBackgroundMaterialKey:String
   ) -> (then)
   > Set TileSizeRuleName(TileSizeRuleName:Name=Select!(
     Size_1_x_1:Name=1_x_1
     Size_2_x_1:Name=2_x_1
     Size_3_x_1:Name=3_x_1
     Size_4_x_1:Name=4_x_1
     Size_5_x_1:Name=5_x_1
     Size_1_x_2:Name=1_x_2
     Size_2_x_2:Name=2_x_2
     Size_3_x_2:Name=3_x_2
     Size_4_x_2:Name=4_x_2
     Size_5_x_2:Name=5_x_2
     Index:EItemShopTileSize Enum=Get TileSize() -> (TileSize=Size_1_x_1)
    ) -> (ReturnValue:Name)) -> (Output_Get:Name=None)
   > n15 = Branch(Condition="Equal (Enum)"(A:E_ItemShopOfferType Enum=Get OfferType() -> (OfferType=NewEnumerator23), B:E_ItemShopOfferType Enum=NewEnumerator22))
   > n16 = Branch(Condition="integer <= integer"(A=Get SmartCardColumns() -> (SmartCardColumns=0), B=2))
   n17 = Append!(
    A="To String (Name)"(InName=Get TileSizeRuleName() -> (TileSizeRuleName:Name=None))
    B=_
    C:String="To String (Integer)"(InInt=Get SmartCardColumns() -> (SmartCardColumns=0))
    D:String=Col
   ) -> (ReturnValue)
   > n18 = Get Data Table Row!(
    execute:Exec
    DataTable:Data Table Object Reference=Get Ruleset_Vehicle() -> (Ruleset_Vehicle:Data Table Object Reference)
    RowName:Name=String To Name(InString=@n17)
   ) -> (then:Exec, RowNotFound:Exec, ReturnValue:S Item Shop Image Ruleset Values Structure)
   > VM_UpdateImageDisplayRuleset!(
    execute
    self:WBP Item Shop Tile Swiping Offer Image Separate Object Reference=Get WBP_ItemShopTile_SwipingOfferImage_Separate()
    ImageDisplayRuleset:S Item Shop Image Ruleset Values Structure=@n18
   ) -> (then)
   > UpdateRulesetDebugText!(
    execute
    self:WBP Item Shop Tile Swiping Offer Image Separate Object Reference=Get WBP_ItemShopTile_SwipingOfferImage_Separate()
    Ruleset:String=Append!(A="Vehicle ", B=@n17) -> (ReturnValue)
   ) -> (then)
   n19 = Get Data Table Row(execute=@n18.RowNotFound, DataTable=Get Ruleset_Vehicle(), RowName=Get TileSizeRuleName() -> (TileSizeRuleName=None))
   > VM_UpdateImageDisplayRuleset(self=Get WBP_ItemShopTile_SwipingOfferImage_Separate(), ImageDisplayRuleset=@n19)
   n20 = Get TileSizeRuleName() -> (TileSizeRuleName=None)
   > UpdateRulesetDebugText(self=Get WBP_ItemShopTile_SwipingOfferImage_Separate(), Ruleset=Append(A="Vehicle ", B="To String (Name)"(InName=@n20)))
   UpdateRulesetDebugText(execute=[@n19.RowNotFound, @n16.else], self=Get WBP_ItemShopTile_SwipingOfferImage_Separate(), Ruleset=Invalid vehicle columns or tile size)
   n21 = Branch(execute=@n15.else, Condition="Equal (Enum)"(A=Get OfferType() -> (OfferType=NewEnumerator23), B=NewEnumerator29))
   n22 = Append!(
    A="To String (Name)"(InName=Get TileSizeRuleName() -> (TileSizeRuleName=None))
    B=_
    C:String="To String (Integer)"(InInt=Get SmartCardColumns() -> (SmartCardColumns=0))
    D:String=Col
   ) -> (ReturnValue)
   n23 = Get Ruleset_SmartCards() -> (Ruleset_SmartCards:Data Table Object Reference)
   > n24 = Get Data Table Row(DataTable=@n23, RowName=String To Name(InString=@n22))
   > VM_UpdateImageDisplayRuleset(self=Get WBP_ItemShopTile_SwipingOfferImage_Separate(), ImageDisplayRuleset=@n24)
   > UpdateRulesetDebugText(self=Get WBP_ItemShopTile_SwipingOfferImage_Separate(), Ruleset=Append!(A="Smart cards ", B=@n22) -> (ReturnValue))
   n25 = Get Data Table Row(execute=@n24.RowNotFound, DataTable=@n23, RowName=Get TileSizeRuleName() -> (TileSizeRuleName=None))
   > VM_UpdateImageDisplayRuleset(self=Get WBP_ItemShopTile_SwipingOfferImage_Separate(), ImageDisplayRuleset=@n25)
   n26 = Get TileSizeRuleName() -> (TileSizeRuleName=None)
   > UpdateRulesetDebugText(self=Get WBP_ItemShopTile_SwipingOfferImage_Separate(), Ruleset=Append(A="Smart cards ", B="To String (Name)"(InName=@n26)))
   UpdateRulesetDebugText(execute=@n25.RowNotFound, self=Get WBP_ItemShopTile_SwipingOfferImage_Separate(), Ruleset=Invalid smart card columns or tile size)
   n27 = Get OfferType() -> (OfferType=NewEnumerator23)
   n28 = Branch(execute=@n21.else, Condition=OR Boolean(A="Equal (Enum)"(A=@n27, B=NewEnumerator3), B="Equal (Enum)"(A=@n27, B=NewEnumerator28)))
   n29 = Get OfferType() -> (OfferType=NewEnumerator23)
   n30 = Append(
    A="To String (Name)"(InName=Get TileSizeRuleName() -> (TileSizeRuleName=None))
    B=Select!(Option 0:String, Option 1:String=_JamTrack, Index:bool="Equal (Enum)"(A=@n29, B=NewEnumerator28)) -> (ReturnValue:String)
   )
   > n31 = Get Data Table Row(DataTable=Get Ruleset_Bundle() -> (Ruleset_Bundle:Data Table Object Reference), RowName=String To Name(InString=@n30))
   > VM_UpdateImageDisplayRuleset(self=Get WBP_ItemShopTile_SwipingOfferImage_Separate(), ImageDisplayRuleset=@n31)
   > UpdateRulesetDebugText(self=Get WBP_ItemShopTile_SwipingOfferImage_Separate(), Ruleset=Append(A="Bundle ", B=@n30))
   n32 = Get Data Table Row DT_ItemShop_ImageRuleset_Bundle!(
    execute:Exec=@n31.RowNotFound
    DataTable:Data Table Object Reference=/FortUILibrary/Systems/ItemShopTile/Rulesets/DT_ItemShop_ImageRuleset_Bundle
    RowName:Name=Get TileSizeRuleName() -> (TileSizeRuleName=None)
   ) -> (then:Exec, RowNotFound:Exec, ReturnValue:S Item Shop Image Ruleset Values Structure) {is=title K2Node_GetDataTableRow}
   > VM_UpdateImageDisplayRuleset(self=Get WBP_ItemShopTile_SwipingOfferImage_Separate(), ImageDisplayRuleset=@n32)
   n33 = Get TileSizeRuleName() -> (TileSizeRuleName=None)
   > UpdateRulesetDebugText(self=Get WBP_ItemShopTile_SwipingOfferImage_Separate(), Ruleset=Append(A="Bundle ", B="To String (Name)"(InName=@n33)))
   UpdateRulesetDebugText(execute=@n32.RowNotFound, self=Get WBP_ItemShopTile_SwipingOfferImage_Separate(), Ruleset=Invalid bundle tile size)
   n34 = Get Ruleset_ScaledStandard() -> (Ruleset_ScaledStandard:Data Table Object Reference)
   n35 = Get Ruleset_ScaledHalfStep() -> (Ruleset_ScaledHalfStep:Data Table Object Reference)
   n36 = Get Ruleset_Bundle()
   n37 = Get Ruleset_Basic() -> (Ruleset_Basic:Data Table Object Reference)
   n38 = Select!(
    NewEnumerator23:Data Table Object Reference=@n34
    NewEnumerator0:Data Table Object Reference=@n35
    NewEnumerator24:Data Table Object Reference=@n34
    NewEnumerator1:Data Table Object Reference=@n35
    NewEnumerator2:Data Table Object Reference=@n35
    NewEnumerator3:Data Table Object Reference=@n36
    NewEnumerator28:Data Table Object Reference=@n36
    NewEnumerator4:Data Table Object Reference=@n34
    NewEnumerator5:Data Table Object Reference=@n35
    NewEnumerator6:Data Table Object Reference=@n34
    NewEnumerator7:Data Table Object Reference=@n34
    NewEnumerator8:Data Table Object Reference=@n37
    NewEnumerator9:Data Table Object Reference=@n35
    NewEnumerator10:Data Table Object Reference=@n35
    NewEnumerator11:Data Table Object Reference=Get Ruleset_FullSquare() -> (Ruleset_FullSquare:Data Table Object Reference)
    NewEnumerator12:Data Table Object Reference=@n35
    NewEnumerator30:Data Table Object Reference=Get Ruleset_Kicks() -> (Ruleset_Kicks:Data Table Object Reference)
    NewEnumerator13:Data Table Object Reference=@n35
    NewEnumerator14:Data Table Object Reference=@n34
    NewEnumerator15:Data Table Object Reference=@n34
    NewEnumerator16:Data Table Object Reference=@n35
    NewEnumerator17:Data Table Object Reference=Get Ruleset_Default() -> (Ruleset_Default:Data Table Object Reference)
    NewEnumerator27:Data Table Object Reference=@n35
    NewEnumerator18:Data Table Object Reference=@n37
    NewEnumerator19:Data Table Object Reference=@n35
    NewEnumerator20:Data Table Object Reference=@n37
    NewEnumerator29:Data Table Object Reference=Get Ruleset_SmartCards()
    NewEnumerator21:Data Table Object Reference=@n34
    NewEnumerator22:Data Table Object Reference=Get Ruleset_Vehicle()
    NewEnumerator25:Data Table Object Reference=@n34
    NewEnumerator26:Data Table Object Reference=@n34
    NewEnumerator31:Data Table Object Reference=Get Ruleset_Sidekicks() -> (Ruleset_Sidekicks:Data Table Object Reference)
    Index:E_ItemShopOfferType Enum=Get OfferType() -> (OfferType=NewEnumerator23)
   ) -> (ReturnValue:Data Table Object Reference)
   n39 = GetRuleForRuleset!(execute=@n28.else, DataTable:Data Table Object Reference=@n38, TileSize:EItemShopTileSize Enum=Get TileSize() -> (TileSize=Size_1_x_1)) -> (then, RulesetValues:S Item Shop Image Ruleset Values Structure)
   > VM_UpdateImageDisplayRuleset(self=Get WBP_ItemShopTile_SwipingOfferImage_Separate(), ImageDisplayRuleset=@n39)
   > UpdateRulesetDebugText(
    self=Get WBP_ItemShopTile_SwipingOfferImage_Separate()
    Ruleset=Split(SourceString=GetObjectName(Object=@n38), InStr=DT_ItemShop_ImageRuleset_, SearchDir=FromStart).RightS
   )
   Get Ruleset_ScaledFull() -> (Ruleset_ScaledFull:Data Table Object Reference)
  ]}
 ToggleRulesetDebug {nodes [
   n1 = ToggleRulesetDebug!() -> (then, ShowRulesetDebug:bool=~) {is=title K2Node_FunctionEntry}
   > n2 = Set ShowRulesetDebug(ShowRulesetDebug:bool=@n1) -> (Output_Get:bool)
   > ToggleRulesetDebug!(
    execute
    self:WBP Item Shop Tile Swiping Offer Image Separate Object Reference=Get WBP_ItemShopTile_SwipingOfferImage_Separate()
    IsEnabled:bool=@n2
   ) -> (then) {is=title K2Node_CallFunction}
  ]}
]
variables [
 BackgroundColor1 {type=LinearColor, default=<rgba 0.2502 0.6867 1 1>}
 BackgroundColor2 {type=LinearColor, default=<rgba 0.0168 0.2122 0.6654 1>}
 BackgroundColor3 {type=LinearColor, default=<rgba 0.2918 0.9473 1 1>}
 BackgroundMaterialOverride {type=MaterialInterface, default=/FortUILibrary/Systems/ItemShopTile/Materials/M_UI_ItemShopTile_Background_Gradient}
 FrameBreak {type=boolean}
 FrameBreakPadding {type=number, default=50}
 OfferType {type=E_ItemShopOfferType, default=Kicks}
 PresentationMaterialOverride {type=MaterialInterface, default=null}
 Ruleset_Basic {type=DataTable, default=/FortUILibrary/Systems/ItemShopTile/Rulesets/DT_ItemShop_ImageRuleset_Basic}
 Ruleset_Bundle {type=DataTable, default=&DT_ItemShop_ImageRuleset_Bundle}
 Ruleset_Default {type=DataTable, default=/FortUILibrary/Systems/ItemShopTile/Rulesets/DT_ItemShop_ImageRuleset_Default}
 Ruleset_FullSquare {type=DataTable, default=/FortUILibrary/Systems/ItemShopTile/Rulesets/DT_ItemShop_ImageRuleset_FullSquare}
 Ruleset_FullTileSpan {type=DataTable, default=/FortUILibrary/Systems/ItemShopTile/Rulesets/DT_ItemShop_ImageRuleset_FullTileSpan}
 Ruleset_Kicks {type=DataTable, default=/FortUILibrary/Systems/ItemShopTile/Rulesets/DT_ItemShop_ImageRuleset_Kicks}
 Ruleset_ScaledFull {type=DataTable, default=/FortUILibrary/Systems/ItemShopTile/Rulesets/DT_ItemShop_ImageRuleset_ScaledFull}
 Ruleset_ScaledHalfStep {type=DataTable, default=/FortUILibrary/Systems/ItemShopTile/Rulesets/DT_ItemShop_ImageRuleset_ScaledHalfStep}
 Ruleset_ScaledStandard {type=DataTable, default=/FortUILibrary/Systems/ItemShopTile/Rulesets/DT_ItemShop_ImageRuleset_ScaledStandard}
 Ruleset_Sidekicks {type=DataTable, default=/FortUILibrary/Systems/ItemShopTile/Rulesets/DT_ItemShop_ImageRuleset_Sidekicks}
 Ruleset_SmartCards {type=DataTable, default=/FortUILibrary/Systems/ItemShopTile/Rulesets/DT_ItemShop_ImageRuleset_SmartCards}
 Ruleset_Vehicle {type=DataTable, default=/FortUILibrary/Systems/ItemShopTile/Rulesets/DT_ItemShop_ImageRuleset_Vehicle}
 ShowRulesetDebug {type=boolean, default=true}
 SmartCardColumns {type=integer}
 Text Backing Color {type=LinearColor, default=<rgba 0 0 0 0>}
 Texture {type=Texture2D, default=/Game/Athena/UI/Challenges/Art/TileImages/T_UI_ChallengeTile_Jonesy}
 TileSize {type=EItemShopTileSize, default=Size_3_x_1}
 TileSizeRuleName {type=string, default=None}
 Use3BackgroundColors {type=boolean}
]
widgets {
 info {inheritedWidgetCount=0, parentClass=&UserWidget, rootWidgetClass=&SizeBox, widgetCount=23}
 widgets [
  SizeBox_Tile {
   widgetClassPath=&SizeBox
   bIsVariable=true
   widget=SizeBox_Tile
   properties {bOverride_HeightOverride=true, bOverride_WidthOverride=true, heightOverride=408, widthOverride=256}
  }
  Overlay_0 {widgetClassPath=&Overlay, parent=SizeBox_Tile, widget=Overlay_0, slot {class=SizeBoxSlot}}
  WBP_ItemShopTile_SwipingOfferImage_Separate {
   widgetClassPath=/ItemShop/MultiProduct/ItemShopTileComponents/WBP_ItemShopTile_SwipingOfferImage_Separate
   bIsVariable=true
   parent=Overlay_0
   widget=WBP_ItemShopTile_SwipingOfferImage_Separate
   slot {class=OverlaySlot, properties {horizontalAlignment=HAlign_Fill, verticalAlignment=VAlign_Fill}}
  }
  HorizontalBox_SmartCards {
   widgetClassPath=&HorizontalBox
   bIsVariable=true
   parent=Overlay_0
   widget=HorizontalBox_SmartCards
   slot {class=OverlaySlot, properties {verticalAlignment=VAlign_Bottom, padding=<ltrb 12 12 12 127>}}
  }
  VerticalBox_62 {
   widgetClassPath=&VerticalBox
   parent=HorizontalBox_SmartCards
   widget=VerticalBox_62
   slot {class=HorizontalBoxSlot, properties {padding=<ltrb 0 0 8 0>}}
  }
  SmartCard {
   widgetClassPath=&Image
   parent=VerticalBox_62
   widget=SmartCard
   properties {brush {
     bIsDynamicallyLoaded=false
     drawAs=RoundedBox
     imageType=NoImage
     mirroring=NoMirror
     resourceName=None
     resourceObject=null
     tiling=NoTile
     imageSize=<96 96>
     margin=<ltrb 0 0 0 0>
     outlineSettings {
      bUseBrushTransparency=false
      roundingType=FixedRadius
      width=0
      color {colorUseRule=UseColor_Specified, specifiedColor=<rgba 0 0 0 0>}
      cornerRadii=<15 15 15 15>
     }
     tintColor {colorUseRule=UseColor_Specified, specifiedColor=<rgba 0 0 0 0.25>}
     uVRegion {bIsValid=false, max=<0 0>, min=<0 0>}
    }}
   slot {class=VerticalBoxSlot, properties {padding=<ltrb 0 0 0 8>}}
  }
  SmartCard_1 {
   widgetClassPath=&Image
   parent=VerticalBox_62
   widget=SmartCard_1
   properties {brush {
     bIsDynamicallyLoaded=false
     drawAs=RoundedBox
     imageType=NoImage
     mirroring=NoMirror
     resourceName=None
     resourceObject=null
     tiling=NoTile
     imageSize=<96 96>
     margin=<ltrb 0 0 0 0>
     outlineSettings {
      bUseBrushTransparency=false
      roundingType=FixedRadius
      width=0
      color {colorUseRule=UseColor_Specified, specifiedColor=<rgba 0 0 0 0>}
      cornerRadii=<15 15 15 15>
     }
     tintColor {colorUseRule=UseColor_Specified, specifiedColor=<rgba 0 0 0 0.25>}
     uVRegion {bIsValid=false, max=<0 0>, min=<0 0>}
    }}
   slot {class=VerticalBoxSlot}
  }
  VerticalBox {
   widgetClassPath=&VerticalBox
   parent=HorizontalBox_SmartCards
   widget=VerticalBox
   slot {class=HorizontalBoxSlot, properties {padding=<ltrb 0 0 8 0>}}
  }
  SmartCard_2 {
   widgetClassPath=&Image
   parent=VerticalBox
   widget=SmartCard_2
   properties {brush {
     bIsDynamicallyLoaded=false
     drawAs=RoundedBox
     imageType=NoImage
     mirroring=NoMirror
     resourceName=None
     resourceObject=null
     tiling=NoTile
     imageSize=<96 96>
     margin=<ltrb 0 0 0 0>
     outlineSettings {
      bUseBrushTransparency=false
      roundingType=FixedRadius
      width=0
      color {colorUseRule=UseColor_Specified, specifiedColor=<rgba 0 0 0 0>}
      cornerRadii=<15 15 15 15>
     }
     tintColor {colorUseRule=UseColor_Specified, specifiedColor=<rgba 0 0 0 0.25>}
     uVRegion {bIsValid=false, max=<0 0>, min=<0 0>}
    }}
   slot {class=VerticalBoxSlot, properties {padding=<ltrb 0 0 0 8>}}
  }
  SmartCard_3 {
   widgetClassPath=&Image
   parent=VerticalBox
   widget=SmartCard_3
   properties {brush {
     bIsDynamicallyLoaded=false
     drawAs=RoundedBox
     imageType=NoImage
     mirroring=NoMirror
     resourceName=None
     resourceObject=null
     tiling=NoTile
     imageSize=<96 96>
     margin=<ltrb 0 0 0 0>
     outlineSettings {
      bUseBrushTransparency=false
      roundingType=FixedRadius
      width=0
      color {colorUseRule=UseColor_Specified, specifiedColor=<rgba 0 0 0 0>}
      cornerRadii=<15 15 15 15>
     }
     tintColor {colorUseRule=UseColor_Specified, specifiedColor=<rgba 0 0 0 0.25>}
     uVRegion {bIsValid=false, max=<0 0>, min=<0 0>}
    }}
   slot {class=VerticalBoxSlot}
  }
  VerticalBox_1 {
   widgetClassPath=&VerticalBox
   parent=HorizontalBox_SmartCards
   widget=VerticalBox_1
   slot {class=HorizontalBoxSlot, properties {padding=<ltrb 0 0 8 0>}}
  }
  SmartCard_4 {
   widgetClassPath=&Image
   parent=VerticalBox_1
   widget=SmartCard_4
   properties {brush {
     bIsDynamicallyLoaded=false
     drawAs=RoundedBox
     imageType=NoImage
     mirroring=NoMirror
     resourceName=None
     resourceObject=null
     tiling=NoTile
     imageSize=<96 96>
     margin=<ltrb 0 0 0 0>
     outlineSettings {
      bUseBrushTransparency=false
      roundingType=FixedRadius
      width=0
      color {colorUseRule=UseColor_Specified, specifiedColor=<rgba 0 0 0 0>}
      cornerRadii=<15 15 15 15>
     }
     tintColor {colorUseRule=UseColor_Specified, specifiedColor=<rgba 0 0 0 0.25>}
     uVRegion {bIsValid=false, max=<0 0>, min=<0 0>}
    }}
   slot {class=VerticalBoxSlot, properties {padding=<ltrb 0 0 0 8>}}
  }
  SmartCard_5 {
   widgetClassPath=&Image
   parent=VerticalBox_1
   widget=SmartCard_5
   properties {brush {
     bIsDynamicallyLoaded=false
     drawAs=RoundedBox
     imageType=NoImage
     mirroring=NoMirror
     resourceName=None
     resourceObject=null
     tiling=NoTile
     imageSize=<96 96>
     margin=<ltrb 0 0 0 0>
     outlineSettings {
      bUseBrushTransparency=false
      roundingType=FixedRadius
      width=0
      color {colorUseRule=UseColor_Specified, specifiedColor=<rgba 0 0 0 0>}
      cornerRadii=<15 15 15 15>
     }
     tintColor {colorUseRule=UseColor_Specified, specifiedColor=<rgba 0 0 0 0.25>}
     uVRegion {bIsValid=false, max=<0 0>, min=<0 0>}
    }}
   slot {class=VerticalBoxSlot}
  }
  VerticalBox_2 {
   widgetClassPath=&VerticalBox
   parent=HorizontalBox_SmartCards
   widget=VerticalBox_2
   slot {class=HorizontalBoxSlot, properties {padding=<ltrb 0 0 8 0>}}
  }
  SmartCard_6 {
   widgetClassPath=&Image
   parent=VerticalBox_2
   widget=SmartCard_6
   properties {brush {
     bIsDynamicallyLoaded=false
     drawAs=RoundedBox
     imageType=NoImage
     mirroring=NoMirror
     resourceName=None
     resourceObject=null
     tiling=NoTile
     imageSize=<96 96>
     margin=<ltrb 0 0 0 0>
     outlineSettings {
      bUseBrushTransparency=false
      roundingType=FixedRadius
      width=0
      color {colorUseRule=UseColor_Specified, specifiedColor=<rgba 0 0 0 0>}
      cornerRadii=<15 15 15 15>
     }
     tintColor {colorUseRule=UseColor_Specified, specifiedColor=<rgba 0 0 0 0.25>}
     uVRegion {bIsValid=false, max=<0 0>, min=<0 0>}
    }}
   slot {class=VerticalBoxSlot, properties {padding=<ltrb 0 0 0 8>}}
  }
  SmartCard_7 {
   widgetClassPath=&Image
   parent=VerticalBox_2
   widget=SmartCard_7
   properties {brush {
     bIsDynamicallyLoaded=false
     drawAs=RoundedBox
     imageType=NoImage
     mirroring=NoMirror
     resourceName=None
     resourceObject=null
     tiling=NoTile
     imageSize=<96 96>
     margin=<ltrb 0 0 0 0>
     outlineSettings {
      bUseBrushTransparency=false
      roundingType=FixedRadius
      width=0
      color {colorUseRule=UseColor_Specified, specifiedColor=<rgba 0 0 0 0>}
      cornerRadii=<15 15 15 15>
     }
     tintColor {colorUseRule=UseColor_Specified, specifiedColor=<rgba 0 0 0 0.25>}
     uVRegion {bIsValid=false, max=<0 0>, min=<0 0>}
    }}
   slot {class=VerticalBoxSlot}
  }
  Border_OfferInfo {
   widgetClassPath=&Border
   bIsVariable=true
   parent=Overlay_0
   widget=Border_OfferInfo
   properties {
    verticalAlignment=VAlign_Bottom
    background {
     bIsDynamicallyLoaded=false
     drawAs=Image
     imageType=FullColor
     mirroring=NoMirror
     resourceName=None
     resourceObject=/FortUILibrary/Systems/ItemShopTile/Materials/MI_UI_ItemShopTile_OfferInfoBacking
     tiling=NoTile
     imageSize=<32 32>
     margin=<ltrb 0 0 0 0>
     outlineSettings {
      bUseBrushTransparency=false
      roundingType=HalfHeightRadius
      width=0
      color {colorUseRule=UseColor_Specified, specifiedColor=<rgba 0 0 0 0>}
      cornerRadii=<0 0 0 0>
     }
     tintColor {colorUseRule=UseColor_Specified, specifiedColor=<rgba 1 1 1 1>}
     uVRegion {bIsValid=false, max=<0 0>, min=<0 0>}
    }
    padding=<ltrb 0 60 0 0>
   }
   slot {class=OverlaySlot, properties {horizontalAlignment=HAlign_Fill, verticalAlignment=VAlign_Bottom}}
  }
  VerticalBox_OfferInfo_1 {
   widgetClassPath=&VerticalBox
   bIsVariable=true
   parent=Border_OfferInfo
   widget=VerticalBox_OfferInfo_1
   slot {class=BorderSlot, properties {verticalAlignment=VAlign_Bottom, padding=<ltrb 0 60 0 0>}}
  }
  Component_Highlight {
   widgetClassPath=/ItemShop/MultiProduct/ItemShopTileComponents/WBP_ItemShopTile_Highlight_2
   bIsVariable=true
   parent=VerticalBox_OfferInfo_1
   widget=Component_Highlight
   properties {bOverride_Cursor=true, highlightText=""}
   slot {class=VerticalBoxSlot, properties {horizontalAlignment=HAlign_Left, verticalAlignment=VAlign_Bottom, padding=<ltrb 12 0 12 6>}}
  }
  Text_Artist {
   widgetClassPath=&CommonTextBlock
   bIsVariable=true
   parent=VerticalBox_OfferInfo_1
   widget=Text_Artist
   properties {
    mobileFontSizeMultiplier=1.2
    scrollStyle=&TextScrollStyle-Base_C
    text=Artist Name
    colorAndOpacity {colorUseRule=UseColor_Specified, specifiedColor=<rgba 1 1 1 0.7>}
    font {
     bForceMonospaced=false
     bMaterialIsStencil=false
     fontMaterial=null
     fontObject=&HeadingNow
     letterSpacing=0
     monospacedWidth=1
     size=12.75
     skewAmount=0
     typefaceFontName=74 Regular
     outlineSettings {
      bApplyOutlineToDropShadows=false
      bMiteredCorners=false
      bSeparateFillAlpha=false
      outlineMaterial=null
      outlineSize=0
      outlineColor=<rgba 0 0 0 1>
     }
    }
    shadowOffset=<0 0>
   }
   slot {class=VerticalBoxSlot, properties {padding=<ltrb 12 0 12 4>}}
  }
  SizeBox_NameText_1 {
   widgetClassPath=&SizeBox
   bIsVariable=true
   parent=VerticalBox_OfferInfo_1
   widget=SizeBox_NameText_1
   properties {bOverride_MaxDesiredHeight=true, maxDesiredHeight=64}
   slot {class=VerticalBoxSlot, properties {padding=<ltrb 12 0 12 4>}}
  }
  Text_Name_1 {
   widgetClassPath=&CommonTextBlock
   bIsVariable=true
   parent=SizeBox_NameText_1
   widget=Text_Name_1
   properties {
    autoWrapText=true
    bIsScrollingEnabled=false
    clipping=OnDemand
    mobileFontSizeMultiplier=1.2
    scrollStyle=&TextScrollStyle-Base_C
    text="[PH] Offer Name Very Long"
    textOverflowPolicy=Ellipsis
    font {
     bForceMonospaced=false
     bMaterialIsStencil=false
     fontMaterial=/FortUILibrary/UIKit/Buttons/MaterialInstances/ItemShopTile/MI_UIKit_ItemShopTile_TitleText
     fontObject=&HeadingNow
     letterSpacing=0
     monospacedWidth=1
     size=15.75
     skewAmount=0
     typefaceFontName=75 Medium
     outlineSettings {
      bApplyOutlineToDropShadows=false
      bMiteredCorners=false
      bSeparateFillAlpha=false
      outlineMaterial=null
      outlineSize=0
      outlineColor=<rgba 0 0 0 1>
     }
    }
    shadowOffset=<0 0>
   }
   slot {class=SizeBoxSlot}
  }
  WBP_ItemShopTile_Price {
   widgetClassPath=/ItemShop/MultiProduct/ItemShopTileComponents/WBP_ItemShopTile_Price
   bIsVariable=true
   parent=VerticalBox_OfferInfo_1
   widget=WBP_ItemShopTile_Price
   slot {class=VerticalBoxSlot, properties {horizontalAlignment=HAlign_Left, verticalAlignment=VAlign_Top, padding=<ltrb 12 0 12 8>}}
  }
 ]
}
