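# Test-zone widget that previews offer image materials on nine item shop tiles of mixed sizes.
# PreConstruct gathers the tiles and passes them to InitializeTiles, which loops over them to update each design-time preview.
# Each tile's dimensions follow its preview size and the Is5ColumnLayout toggle.
# EnableApplyToAll swaps each tile's own preview material for the shared MaterialToPreview, while hover state stays per tile.
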WidgetBlueprint /ItemShop/A_TestWidgets/A_OfferImage_TestZone
graphs [
 EventGraph {nodes [
   n1 = Event PreConstruct()
   n2 = Set TileWidgets(
    execute=[@n1.then, @n3.then]
    TileWidgets:Array of WBP Item Shop Tile Default Object References=Make Array!(
     "[0]":WBP Item Shop Tile Default Object Reference=Get Tile_TripleWide() -> (Tile_TripleWide:WBP Item Shop Tile Default Object Reference)
     "[1]":WBP Item Shop Tile Default Object Reference=Get Tile_DoubleWide() -> (Tile_DoubleWide:WBP Item Shop Tile Default Object Reference)
     "[2]":WBP Item Shop Tile Default Object Reference=Get Tile_Normal() -> (Tile_Normal:WBP Item Shop Tile Default Object Reference)
     "[3]":WBP Item Shop Tile Default Object Reference=Get Tile_Small() -> (Tile_Small:WBP Item Shop Tile Default Object Reference)
     "[4]":WBP Item Shop Tile Default Object Reference=Get Tile_Small_2() -> (Tile_Small_2:WBP Item Shop Tile Default Object Reference)
     "[5]":WBP Item Shop Tile Default Object Reference=Get Tile_DoubleWide_2() -> (Tile_DoubleWide_2:WBP Item Shop Tile Default Object Reference)
     "[6]":WBP Item Shop Tile Default Object Reference=Get Tile_Normal_2() -> (Tile_Normal_2:WBP Item Shop Tile Default Object Reference)
     "[7]":WBP Item Shop Tile Default Object Reference=Get Tile_Small_3() -> (Tile_Small_3:WBP Item Shop Tile Default Object Reference)
     "[8]":WBP Item Shop Tile Default Object Reference=Get Tile_Small_4() -> (Tile_Small_4:WBP Item Shop Tile Default Object Reference)
    ) -> (Array:Array of WBP Item Shop Tile Default Object References)
   ) -> (Output_Get:Array of WBP Item Shop Tile Default Object References)
   > InitializeTiles!(execute, self, Array:Array of WBP Item Shop Tile Default Object References=@n2) -> (then)
   n3 = Set MaterialArray(MaterialArray:Array of Fort Item Shop Loaded Presentation Structures=Make Array!(
     "[0]_Texture":Texture 2D Object Reference
     "[0]_Material":Material Interface Object Reference=Get MaterialToPreview() -> (MaterialToPreview:Material Interface Object Reference)
    ) -> (Array:Array of Fort Item Shop Loaded Presentation Structures)) -> (Output_Get:Array of Fort Item Shop Loaded Presentation Structures)
  ]}
 InitializeTiles {nodes [
   n1 = InitializeTiles!() -> (then, Array:Array of WBP Item Shop Tile Default Object References) {is=title K2Node_FunctionEntry}
   > n2 = For Each Loop(Array:Array of WBP Item Shop Tile Default Object References=@n1) -> (Array Element:WBP Item Shop Tile Default Object Reference)
   > Set LoopTile(LoopTile:WBP Item Shop Tile Default Object Reference=@n2) -> (Output_Get:WBP Item Shop Tile Default Object Reference)
   n3 = NOT Boolean(A=Get Is5ColumnLayout() -> (Is5ColumnLayout:bool))
   n4 = Get LoopTile() -> (LoopTile:WBP Item Shop Tile Default Object Reference)
   > SetTileDimensions!(
    execute
    self:WBP Item Shop Tile Base Object Reference=Get LoopTile()
    Dimensions:Vector 2D Structure=Select!(
     Size_1_x_1:Vector 2D Structure=Select!(Option 0:Vector 2D Structure=(X=220,Y=264), Option 1:Vector 2D Structure=(X=281,Y=337), Index:bool=@n3) -> (ReturnValue)
     Size_2_x_1:Vector 2D Structure
     Size_3_x_1:Vector 2D Structure
     Size_4_x_1:Vector 2D Structure
     Size_5_x_1:Vector 2D Structure
     Size_1_x_2:Vector 2D Structure=Select(Option 0=(X=220,Y=552), Option 1=(X=281,Y=694), Index=@n3)
     Size_2_x_2:Vector 2D Structure=Select(Option 0=(X=464,Y=552), Option 1=(X=586,Y=694), Index=@n3)
     Size_3_x_2:Vector 2D Structure=Select(Option 0=(X=708,Y=552), Option 1=(X=891,Y=694), Index=@n3)
     Size_4_x_2:Vector 2D Structure
     Size_5_x_2:Vector 2D Structure=Select(Option 0=(X=1196,Y=552), Option 1=(X=1453,Y=694), Index=@n3)
     Index:EItemShopTileSize Enum=Get DesignTimePreviewTileSize!(self:WBP Item Shop Tile Default Object Reference=@n4) -> (DesignTimePreviewTileSize:EItemShopTileSize Enum=Size_1_x_1)
    ) -> (ReturnValue:Vector 2D Structure)
   ) -> (then)
   n5 = Get LoopTile()
   > SetDesignTimePreview!(
    execute
    self:WBP Item Shop Tile Default Object Reference=Get LoopTile()
    PresentationMaterial:Material Interface Object Reference=Select(
     Option 0:Material Interface Object Reference=Get DesignTimeMaterialPreview!(self:WBP Item Shop Tile Default Object Reference=@n5) -> (DesignTimeMaterialPreview:Material Interface Object Reference)
     Option 1:Material Interface Object Reference=Get MaterialToPreview()
     Index=Get EnableApplyToAll() -> (EnableApplyToAll:bool)
    ) -> (ReturnValue:Material Interface Object Reference)
    TileSize:EItemShopTileSize Enum=Get DesignTimePreviewTileSize(self=Get LoopTile()) -> (DesignTimePreviewTileSize=Size_1_x_1)
    IsHovered:bool=Get DesignTimePreviewHoveredState!(self:WBP Item Shop Tile Default Object Reference=Get LoopTile()) -> (DesignTimePreviewHoveredState:bool)
   ) -> (then)
  ]}
]
variables [
 EnableApplyToAll {type=boolean}
 Is5ColumnLayout {type=boolean}
 MaterialArray {type="array<FortItemShopLoadedPresentation>", default []}
 MaterialToPreview {type=MaterialInterface, default=/OfferCatalog/MI_OfferImages/MI_0_OfferImage_FNCSSeries}
 TileWidgets {type="array<WBP_ItemShopTile_Default_C>", default []}
]
widgets {
 info {inheritedWidgetCount=0, parentClass=&UserWidget, rootWidgetClass=&Overlay, widgetCount=22}
 widgets [
  All {widgetClassPath=&Overlay, widget=All}
  Backing {
   widgetClassPath=&Image
   bIsVariable=true
   parent=All
   widget=Backing
   properties {colorAndOpacity=<rgba 0.0325 0.0325 0.0325 1>}
   slot {class=OverlaySlot, properties {horizontalAlignment=HAlign_Fill, verticalAlignment=VAlign_Fill}}
  }
  Content {
   widgetClassPath=&VerticalBox
   parent=All
   widget=Content
   slot {class=OverlaySlot, properties {horizontalAlignment=HAlign_Fill, verticalAlignment=VAlign_Fill, padding=<ltrb 256 256 256 256>}}
  }
  Tile_TripleWide {
   widgetClassPath=/ItemShop/MultiProduct/WBP_ItemShopTile_Default
   bIsVariable=true
   parent=Content
   widget=Tile_TripleWide
   properties {designTimePreviewTileSize=Size_3_x_2}
   slot {class=VerticalBoxSlot, properties {horizontalAlignment=HAlign_Left}}
  }
  BottomTiles_1 {
   widgetClassPath=&HorizontalBox
   parent=Content
   widget=BottomTiles_1
   slot {class=VerticalBoxSlot, properties {padding=<ltrb 0 24 0 0>}}
  }
  Tile_DoubleWide {
   widgetClassPath=&WBP_ItemShopTile_Default
   bIsVariable=true
   parent=BottomTiles_1
   widget=Tile_DoubleWide
   properties {designTimePreviewTileSize=Size_2_x_2}
   slot {class=HorizontalBoxSlot, properties {horizontalAlignment=HAlign_Left, verticalAlignment=VAlign_Top}}
  }
  Spacer {widgetClassPath=&Spacer, parent=BottomTiles_1, widget=Spacer, properties {size=<24 1>}, slot {class=HorizontalBoxSlot}}
  Tile_Normal {
   widgetClassPath=&WBP_ItemShopTile_Default
   bIsVariable=true
   parent=BottomTiles_1
   widget=Tile_Normal
   slot {class=HorizontalBoxSlot, properties {horizontalAlignment=HAlign_Left, verticalAlignment=VAlign_Top}}
  }
  Spacer_1 {widgetClassPath=&Spacer, parent=BottomTiles_1, widget=Spacer_1, properties {size=<24 1>}, slot {class=HorizontalBoxSlot}}
  VerticalBox_1 {widgetClassPath=&VerticalBox, parent=BottomTiles_1, widget=VerticalBox_1, slot {class=HorizontalBoxSlot}}
  Tile_Small {
   widgetClassPath=&WBP_ItemShopTile_Default
   bIsVariable=true
   parent=VerticalBox_1
   widget=Tile_Small
   properties {1x1TileSizeRatio=<1 1.2>}
   slot {class=VerticalBoxSlot, properties {horizontalAlignment=HAlign_Left, verticalAlignment=VAlign_Top}}
  }
  Spacer_4 {widgetClassPath=&Spacer, parent=VerticalBox_1, widget=Spacer_4, properties {size=<24 24>}, slot {class=VerticalBoxSlot}}
  Tile_Small_2 {
   widgetClassPath=&WBP_ItemShopTile_Default
   bIsVariable=true
   parent=VerticalBox_1
   widget=Tile_Small_2
   properties {1x1TileSizeRatio=<1 1.2>}
   slot {class=VerticalBoxSlot, properties {horizontalAlignment=HAlign_Left, verticalAlignment=VAlign_Top}}
  }
  BottomTiles_2 {
   widgetClassPath=&HorizontalBox
   parent=Content
   widget=BottomTiles_2
   slot {class=VerticalBoxSlot, properties {verticalAlignment=VAlign_Top, padding=<ltrb 0 24 0 0>}}
  }
  Tile_DoubleWide_2 {
   widgetClassPath=&WBP_ItemShopTile_Default
   bIsVariable=true
   parent=BottomTiles_2
   widget=Tile_DoubleWide_2
   properties {designTimePreviewTileSize=Size_2_x_2}
   slot {class=HorizontalBoxSlot, properties {horizontalAlignment=HAlign_Left, verticalAlignment=VAlign_Top}}
  }
  Spacer_2 {widgetClassPath=&Spacer, parent=BottomTiles_2, widget=Spacer_2, properties {size=<24 1>}, slot {class=HorizontalBoxSlot}}
  Tile_Normal_2 {
   widgetClassPath=&WBP_ItemShopTile_Default
   bIsVariable=true
   parent=BottomTiles_2
   widget=Tile_Normal_2
   slot {class=HorizontalBoxSlot, properties {horizontalAlignment=HAlign_Left, verticalAlignment=VAlign_Top}}
  }
  Spacer_3 {widgetClassPath=&Spacer, parent=BottomTiles_2, widget=Spacer_3, properties {size=<24 1>}, slot {class=HorizontalBoxSlot}}
  VerticalBox_2 {widgetClassPath=&VerticalBox, parent=BottomTiles_2, widget=VerticalBox_2, slot {class=HorizontalBoxSlot}}
  Tile_Small_3 {
   widgetClassPath=&WBP_ItemShopTile_Default
   bIsVariable=true
   parent=VerticalBox_2
   widget=Tile_Small_3
   properties {1x1TileSizeRatio=<1 1.2>}
   slot {class=VerticalBoxSlot, properties {horizontalAlignment=HAlign_Left, verticalAlignment=VAlign_Top}}
  }
  Spacer_5 {widgetClassPath=&Spacer, parent=VerticalBox_2, widget=Spacer_5, properties {size=<24 24>}, slot {class=VerticalBoxSlot}}
  Tile_Small_4 {
   widgetClassPath=&WBP_ItemShopTile_Default
   bIsVariable=true
   parent=VerticalBox_2
   widget=Tile_Small_4
   properties {1x1TileSizeRatio=<1 1.2>}
   slot {class=VerticalBoxSlot, properties {horizontalAlignment=HAlign_Left, verticalAlignment=VAlign_Top}}
  }
 ]
}
